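# Material function for the DracoDueler character that outputs an animated blink mask within a configurable circle.
# The inverted circle SDF drives a BumpOffset, whose parallax-shifted U coordinate forms the eyelid gradient.
# Time-based sine waves, shaped by SmoothStep with speed, delay and threshold inputs, briefly shift that gradient.
# Two similar eyelid-edge layers with different softness and travel are summed, then masked by the circle fill.
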MaterialFunction /BRCosmetics/Characters/Player/Male/Medium/Bodies/M_MED_DracoDueler/Materials/MaterialFunctions/MF_DracoDueler_Blink
graph {nodes [
  n1 = NamedRerouteUsage()
  n2 = TextureCoordinate()
  n3 = Constant2Vector() {g=0.475, r=0.03}
  n4 = MaterialFunctionCall(
   UVs=Add(
    A=FunctionInput(Preview=@n2) {bUsePreviewValueAsDefault=true, inputName=UV, inputType=FunctionInput_Vector2}
    B=FunctionInput(Preview=@n3) {bUsePreviewValueAsDefault=true, inputName=CircleMaskPosition, inputType=FunctionInput_Vector2}
   )
   Size=FunctionInput(Preview=Constant() {r=0.09}) {bUsePreviewValueAsDefault=true, inputName=CircleMaskRadius, inputType=FunctionInput_Scalar}
  ) {materialFunction=&MF_UI_Circle}
  n5 = OneMinus(None=@n4.SDF)
  n6 = BumpOffset(Height=Multiply(A=Subtract(A=@n5)) {constB=10})
  n7 = ComponentMask(None=@n6) {r=true}
  n8 = NamedRerouteUsage()
  n9 = Add(A=Time(), B=NamedRerouteUsage()) {constB=0}
  n10 = NamedRerouteUsage()
  n11 = Add(A=Sine(None=Multiply(A=@n9, B=@n10)))
  n12 = Constant() {r=0.1}
  n13 = Multiply(A=SmoothStep(Min=Subtract(A=@n8) {constB=0.02}, Value=Multiply(A=@n11) {constB=0.5}) {constMin=0.35}, B=@n12) {constB=0.08}
  n14 = Multiply(A=SmoothStep(Min=@n1, Max=Add(A=@n1) {constB=0.03}, Value=Add(A=@n7, B=@n13)) {constMax=0.52, constMin=0.51}) {constB=0.9}
  n15 = NamedRerouteUsage()
  n16 = Time()
  n17 = NamedRerouteUsage()
  n18 = Sine(None=Multiply(A=Add(A=@n16, B=@n17) {constB=0}, B=NamedRerouteUsage()))
  n19 = SmoothStep(Min=NamedRerouteUsage(), Value=Multiply(A=Add(A=@n18)) {constB=0.5}) {constMin=0.35}
  n20 = SmoothStep(Min=@n15, Max=Add(A=@n15) {constB=0}, Value=Add(A=ComponentMask(None=@n6) {r=true}, B=Multiply(A=@n19, B=@n12) {constB=0.1})) {constMax=0.55, constMin=0.51}
  FunctionOutput(None=Multiply(A=Add(A=@n14, B=@n20), B=@n4.Fill) {constB=0.2})
  NamedRerouteDeclaration(None=FunctionInput(Preview=Constant() {r=0.997}) {bUsePreviewValueAsDefault=true, inputName=BlinkingThreshhold, inputType=FunctionInput_Scalar}) {name=ParameterBlinkThreshhold, nodeColor=<rgba 0.4766 0 1 1>}
  NamedRerouteDeclaration(None=FunctionInput(Preview=Constant()) {bUsePreviewValueAsDefault=true, inputName=BlinkingDelay, inputType=FunctionInput_Scalar}) {name=ParameterBlinkDelay, nodeColor=<rgba 0 0.5547 1 1>}
  NamedRerouteDeclaration(None=FunctionInput(Preview=Constant() {r=0.51}) {bUsePreviewValueAsDefault=true, inputName=EyelidPositionMin, inputType=FunctionInput_Scalar}) {name=ParameterEyelidPositionMin, nodeColor=<rgba 0.3828 1 0 1>}
  NamedRerouteDeclaration(None=FunctionInput(Preview=Constant() {r=0.15}) {bUsePreviewValueAsDefault=true, inputName=BlinkSpeed, inputType=FunctionInput_Scalar}) {name=ParameterBlinkSpeed, nodeColor=<rgba 0.1016 1 0 1>}
 ]}
