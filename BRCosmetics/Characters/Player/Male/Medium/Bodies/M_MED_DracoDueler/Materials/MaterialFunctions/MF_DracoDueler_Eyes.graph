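# Material function for the Draco Dueler character's eyes that adds an animated blink to the material's emissive color.
# Two MF_DracoDueler_Blink calls build left and right lid masks from offset circles, sharing blink speed and threshold parameters.
# The inverted sum of the masks drives a lerp on emissive, darkening the lid areas.
# The UseEyeAnimation static switch selects this result or passes the input attributes through unchanged.
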MaterialFunction /BRCosmetics/Characters/Player/Male/Medium/Bodies/M_MED_DracoDueler/Materials/MaterialFunctions/MF_DracoDueler_Eyes
graph {nodes [
  n1 = FunctionInput() {inputType=FunctionInput_MaterialAttributes}
  n2 = GetMaterialAttributes(None=@n1) {attributeGetTypes [#1]}
  n3 = NamedRerouteUsage()
  n4 = NamedRerouteUsage()
  n5 = LinearInterpolate(B=NamedRerouteUsage(), Alpha=OneMinus(None=Add(A=@n3, B=@n4)))
  FunctionOutput(None=StaticSwitchParameter(
    True=SetMaterialAttributes!(MaterialAttributes=@n2.MaterialAttributes, EmissiveColor=@n5) {attributeSetTypes [#1]}
    False=@n1
   ) {parameterName=UseEyeAnimation})
  n6 = ScalarParameter() {defaultValue=0.998, group=Eyes, parameterName=BlinkingThreshold, sortPriority=36}
  n7 = ScalarParameter() {defaultValue=0.2, parameterName=BlinkSpeed}
  NamedRerouteDeclaration(None=MaterialFunctionCall!(
    CircleMaskPosition=Constant2Vector() {g=0.475, r=0.03}
    CircleMaskRadius=Constant() {r=0.09}
    BlinkingThreshhold=@n6
    BlinkingDelay=Constant() {r=0.02}
    EyelidPositionMin=Constant() {r=0.51}
    BlinkSpeed=@n7
    UV
   ) {materialFunction=&/MF_DracoDueler_Blink}.Result) {name=LeftLidMask, nodeColor=<rgba 0 0.7187 1 1>}
  NamedRerouteDeclaration(None=MaterialFunctionCall(
    CircleMaskPosition=Constant2Vector() {g=0.475, r=-0.07}
    CircleMaskRadius=Constant() {r=0.09}
    BlinkingThreshhold=@n6
    EyelidPositionMin=Constant() {r=0.61}
    BlinkSpeed=@n7
   ) {materialFunction=&MF_DracoDueler_Blink}.Result) {name=RightLidMask, nodeColor=<rgba 0.9922 0 1 1>}
  NamedRerouteDeclaration(None=@n2.EmissiveColor) {name=EmissiveColor, nodeColor=<rgba 0.4531 0 1 1>}
 ]}
